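# Physics asset for the Galaxy vehicle body cosmetic mesh, giving it a simplified collision shape.
# A single box body on the root bone covers the whole vehicle, centered slightly forward and raised above the origin.
# It has no other bodies and no constraints, so the body simulates as one rigid piece.
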PhysicsAsset /VehicleCosmetics/Bodies/Galaxy/Meshes/Body_Galaxy_PhysicsDM
bodies {root [{shapeType=Box, extentX=444, extentY=216, extentZ=132, shapeName=root_box, center=<16.5 0 100>}]}
constraints []
